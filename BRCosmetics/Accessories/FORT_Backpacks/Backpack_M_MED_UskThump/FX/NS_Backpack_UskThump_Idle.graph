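# Niagara system for the UskThump backpack that sloshes coffee-like liquid from a socket as the wearer moves and acts.
# Five sprite emitters (waves, splashes, sprint drops) spawn per frame, with probability gated by landing, crouch and firing flags or count scaled by owner speed and a system curve.
# Socket sampling and inherited movement keep particles near the pack; skydiving disables drop gravity.
# Waves randomly mirror via negative X scale and tint by time of day.
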NiagaraSystem /BRCosmetics/Accessories/FORT_Backpacks/Backpack_M_MED_UskThump/FX/NS_Backpack_UskThump_Idle
dependencies {
 usedDataInterfaces [
  &NiagaraDataInterfaceCurve
  &NiagaraDataInterfaceColorCurve
  &NiagaraDataInterfaceVector2DCurve
  &NiagaraDataInterfaceSocketReader
  &NiagaraDataInterfaceArrayFloat
 ]
 usedDynamicInputs [&MergeHelper_ColorSwitch, &Multiply_Vector]
 usedModules [
  &SystemState
  &self:SystemSpawnScript.SystemScriptSource.SystemScriptGraph.NiagaraNodeAssignment_4.SetVariables_#1
  &self:SystemSpawnScript.SystemScriptSource.SystemScriptGraph.NiagaraNodeAssignment_3.SetVariables_#2
  &self:SystemSpawnScript.SystemScriptSource.SystemScriptGraph.NiagaraNodeAssignment_2.SetVariables_#3
  &self:SystemSpawnScript.SystemScriptSource.SystemScriptGraph.NiagaraNodeAssignment_1.SetVariables_#4
  &Timeline
 ]
}
emitters {
 Coffee_Splash {
  emitterUpdateScript [
   &EmitterState()
   &SpawnPerFrame(
    Spawn Count:NiagaraInt32=&MakeIntFromFloat(Float=&Multiply_Float(
      A=&Lerp_Float(
       A=0
       B=1
       Alpha=&ClampFloat(Float=&Divide_Float(A=&VectorLength(Vector=$Engine.Owner.Velocity), B=500))
      )
      B=$System.SpawnRateCurve
     ))
    Spawn Probability:NiagaraFloat=&MakeCustomFloatFromBool(
     Bool=&SetBoolByFloatComparison(A=2000, B=&VectorLength(Vector=$Engine.Owner.Velocity))
     True Float=0.2
     False Float=0
    )
    Use Spawn Probability=true
   ) {module=SpawnPerFrame003}
   &SpawnPerFrame(
    Spawn Probability:NiagaraFloat=&MakeCustomFloatFromBool(Bool=$User.IsWeaponFired, True Float=0.3, False Float=0)
    Use Spawn Probability=true
   ) {module=SpawnPerFrame002}
   &SpawnPerFrame(
    Spawn Probability:NiagaraFloat=&MakeCustomFloatFromBool(Bool=$User.IsCrouched, True Float=0.75, False Float=0)
    Use Spawn Probability=true
   ) {module=SpawnPerFrame001}
   &SpawnPerFrame(
    Spawn Probability:NiagaraFloat=&MakeCustomFloatFromBool(Bool=$User.bPawnLanded, True Float=1, False Float=0)
    Use Spawn Probability=true
   )
  ]
  particleSpawnScript [
   &InitializeParticle(
    Color Mode=::NewEnumerator0
    Lifetime=&RandomRangeFloat(Minimum=0.3, Maximum=0.5)
    Sprite Rotation Mode=::NewEnumerator2
    Sprite Size Mode=::NewEnumerator3
    Uniform Sprite Size=&Multiply_Float(A=&RandomRangeFloat(Minimum=4, Maximum=6.5), B=1.2)
    UsePositionOffset=false
    Write Material Random=true
   )
   &CameraOffset(Camera Offset Amount=2) {module=CameraOffset001}
   &ShapeLocation(Disc Coverage=1, MaskSpawnGroup=false, Override Seed=false, Ring Radius=5, Shape Primitive=::NewEnumerator4, UseOwnerScale=true) {module=ShapeLocation001}
   &AddVelocity(
    Cone Axis=<0 0 1>
    Module.UseConeFalloff=false
    Module.UseVelDistribution=false
    Override Seed=false
    Rotation Coordinate Space=::World
    Velocity Mode=::NewEnumerator2
    Velocity Speed=&RandomRangeFloat(Minimum=100, Maximum=150)
   ) {module=AddVelocity002}
  ]
  particleUpdateScript [
   &ParticleState()
   &SocketLocation(SamplePosOffset=true, Socket Reader=$System.Socket Reader, Socket Sampling Mode=::NewEnumerator3, VelocitySpeedLimit=false)
   &ScaleSpriteSize(Scale Factor=&Vector2DFromFloat(Value=&FloatFromCurve(FloatCurve=DataInterface() {propertyValues {Curve {keys [<at 0 0>,<at 0.5 1>]}}}, Scale Curve=2))) {module=ScaleSpriteSize001}
   &GravityForce(Gravity=<0 0 -600>) {module=GravityForce001}
   &SolveForcesAndVelocity(Clamp Velocity=false, Limit Acceleration=false) {module=SolveForcesAndVelocity001}
   &ScaleSpriteSize(Scale Factor=&Vector2DFromFloat(Value=&FloatFromCurve(FloatCurve=DataInterface() {propertyValues {Curve {keys [<at 0 0>,<at 1 1>]}}}, Scale Curve=1.5)))
   &DynamicMaterialParameters(
    Index 0 Param 1=&RandomRangeFloat(Minimum=0, Maximum=1)
    Index 0 Param 3=&FloatFromCurve(FloatCurve=DataInterface() {propertyValues {Curve {keys [{interpMode=RCIM_Cubic, tangentMode=RCTM_User, time=0, value=0}, {interpMode=RCIM_Cubic, tangentMode=RCTM_User, time=1, value=1}]}}})
    Index 0 Param 4=&FloatFromCurve(FloatCurve=DataInterface() {propertyValues {Curve {keys [<at 0 0.2 RCIM_Cubic>, {arriveTangent=1.4369, interpMode=RCIM_Cubic, leaveTangent=1.4369, tangentMode=RCTM_User, time=1, value=1}]}}})
    Module.Param0WriteEnabled=true
    Module.Param1WriteEnabled=false
    Module.Param2WriteEnabled=true
    Module.Param3WriteEnabled=true
   ) {module=DynamicMaterialParameters001}
   &ScaleColor(
    Scale Alpha=10
    Scale RGB=&MakeVectorFromLinearColorRGB(Color=&ColorFromCurve(ColorCurve=DataInterface() {propertyValues {
        AlphaCurve {keys [<at 0 1 RCIM_Cubic>]}
        BlueCurve {keys [<at 0 0.0003 RCIM_Cubic>,<at 1 0.007>]}
        GreenCurve {keys [<at 0 0.0018 RCIM_Cubic>,<at 1 0.0469>]}
        RedCurve {keys [<at 0 0.0802 RCIM_Cubic>,<at 1 0.1766>]}
       }}))
    ScaleA=true
    ScaleRGB=true
   )
  ]
  rendererClasses [&NiagaraSpriteRendererProperties]
  renderers [{rendererClass=&NiagaraSpriteRendererProperties}]
 }
 Coffee_Waves {
  emitterUpdateScript [
   &EmitterState()
   &SpawnPerFrame(
    Spawn Probability:NiagaraFloat=&MakeCustomFloatFromBool(Bool=$User.bPawnLanded, True Float=1, False Float=0)
    Use Spawn Probability=true
   )
   &SpawnPerFrame(
    Spawn Group=1
    Spawn Probability:NiagaraFloat=&MakeCustomFloatFromBool(Bool=$User.IsCrouched, True Float=0.15, False Float=0)
    Use Spawn Probability=true
   ) {module=SpawnPerFrame001}
   &SpawnPerFrame(
    Spawn Count:NiagaraInt32=&MakeIntFromFloat(Float=&Multiply_Float(
      A=&Lerp_Float(
       A=0
       B=1
       Alpha=&ClampFloat(Float=&Divide_Float(A=&VectorLength(Vector=$Engine.Owner.Velocity), B=500))
      )
      B=$System.SpawnRateCurve
     ))
    Spawn Probability:NiagaraFloat=&MakeCustomFloatFromBool(
     Bool=&SetBoolByFloatComparison(A=2000, B=&VectorLength(Vector=$Engine.Owner.Velocity))
     True Float=&MakeCustomFloatFromBool(Bool=$User.IsSkydiving, True Float=0, False Float=0.2)
     False Float=0
    )
    Use Spawn Probability=true
   ) {module=SpawnPerFrame003}
  ]
  particleSpawnScript [
   &InitializeParticle(
    Color=<rgba 0.0976 0.0232 0.0307 1>
    Lifetime Max=0.5
    Lifetime Min=0.25
    Lifetime Mode=::NewEnumerator1
    Sprite Size=&Multiply_Vector2D_ByFloat(Float=0.5, Vector2D=&RandomRangeVector2D(Minimum=<10 2>, Maximum=<12 5>))
    Sprite Size Mode=::NewEnumerator4
    UsePositionOffset=true
    Write Material Random=true
   )
   /SetVariables(Particles.IsLHS:NiagaraBool=&self:ScratchDynamicInput_02()) {isSetParameters=true}
   /SetVariables(Particles.PivotOffset:Vector2f=&MakeCustomVector2DFromBool(Bool=$Particles.IsLHS, True Vector2D=<0.75 1>, False Vector2D=<0.25 1>)) {isSetParameters=true}
   &SubUVAnimation(End Frame=1, SubUV Animation Mode:ENiagara_SubUVLookupMode=::NewEnumerator2)
   &CameraOffset(Camera Offset Amount=2)
  ]
  particleUpdateScript [
   &ParticleState()
   &SocketLocation(SamplePosOffset=false, Socket Reader=$System.Socket Reader, Socket Sampling Mode=::NewEnumerator3, VelocitySpeedLimit=false)
   /SetVariables() {isSetParameters=true}
   /SetVariables(Particles.SpriteAlignment=<0 0 1>) {isSetParameters=true}
   &DynamicMaterialParameters(
    Index 0 Param 1=&Lerp_Float(
     A=0.3
     B=1
     Alpha=&FloatFromCurve(FloatCurve=DataInterface() {propertyValues {Curve {keys [
          {arriveTangent=0.1629, interpMode=RCIM_Cubic, time=0, value=0.975}
          {arriveTangent=0.0278, interpMode=RCIM_Cubic, leaveTangent=0.0278, time=0.2, value=0.025}
          <at 0.9 1 RCIM_Cubic>
         ]}}})
    )
    Module.Param0WriteEnabled=true
    Module.Param1WriteEnabled=true
    Module.Param2WriteEnabled=true
    Module.Param3WriteEnabled=true
   )
   &ScaleSpriteSize(
    Non-Uniform Curve Sprite Scale=DataInterface() {propertyValues {
      XCurve {keys [
        {arriveTangent=2.8681, interpMode=RCIM_Cubic, leaveTangent=2.8681, tangentMode=RCTM_User, time=0, value=1}
        {arriveTangent=1.25, interpMode=RCIM_Cubic, leaveTangent=1.25, time=0.4, value=1.75}
        <at 1 2.25 RCIM_Cubic>
       ]}
      YCurve {keys [
        <at 0 1 RCIM_Cubic>
        {arriveTangent=0.7, interpMode=RCIM_Cubic, leaveTangent=0.7, time=0.1, value=2.425}
        {arriveTangent=-1.5833, interpMode=RCIM_Cubic, leaveTangent=-1.5833, time=0.5, value=1.35}
        <at 1 1 RCIM_Cubic>
       ]}
     }}
    Scale Sprite Size Mode=::NewEnumerator3
   )
   &ScaleSpriteSize(
    Scale Factor=&MakeVector2D(X=&MakeCustomFloatFromBool(Bool=$Particles.IsLHS, True Float=-1, False Float=1), Y=1)
    Scale Sprite Size Mode=::NewEnumerator2
   ) {module=ScaleSpriteSize001}
   &ScaleColor(
    Scale RGB=&Multiply_VectorByFloat(
     Vector=&FN_Select_Vector_from_Time_Of_Day(Morning=<1 0.6 0.6>, Day=<1 0.9 0.9>, Evening=<1 0.7 0.7>, Night=<0.75 1 1>, Evening Threshold=19)
     Float=&FNTimeOfDayScalar(Night Scale=0.2, Morning Scale=0.65, Day Scale=0.8, Evening Scale=0.4)
    )
    ScaleA=true
    ScaleRGB=true
   )
  ]
  rendererClasses [&NiagaraSpriteRendererProperties]
  renderers [{rendererClass=&NiagaraSpriteRendererProperties}]
 }
 Splash {
  emitterUpdateScript [
   &EmitterState()
   &SpawnPerFrame(
    Spawn Probability:NiagaraFloat=&MakeCustomFloatFromBool(Bool=$User.IsWeaponFired, True Float=0.3, False Float=0)
    Use Spawn Probability=true
   ) {module=SpawnPerFrame002}
   &SpawnPerFrame(
    Spawn Probability:NiagaraFloat=&MakeCustomFloatFromBool(Bool=$User.IsCrouched, True Float=0.75, False Float=0)
    Use Spawn Probability=true
   ) {module=SpawnPerFrame001}
   &SpawnPerFrame(
    Spawn Count=2
    Spawn Probability:NiagaraFloat=&MakeCustomFloatFromBool(Bool=$User.bPawnLanded, True Float=1, False Float=0)
    Use Spawn Probability=true
   )
  ]
  particleSpawnScript [
   &InitializeParticle(
    Color Mode=::NewEnumerator0
    Lifetime=&RandomRangeFloat(Minimum=0.3, Maximum=0.5)
    Sprite Rotation Mode=::NewEnumerator2
    Sprite Size Mode=::NewEnumerator3
    Uniform Sprite Size=&Multiply_Float(A=&RandomRangeFloat(Minimum=4, Maximum=6.5), B=1.25)
    UsePositionOffset=false
    Write Material Random=true
   )
   &SocketLocation(
    SamplePosOffset=true
    Sampled Position Offset=<0 0 -5>
    Socket Reader=$System.Socket Reader
    Socket Sampling Mode=::NewEnumerator3
    VelocitySpeedLimit=false
   )
   &ShapeLocation(Disc Coverage=1, MaskSpawnGroup=false, Override Seed=false, Ring Radius=2, Shape Primitive=::NewEnumerator4, UseOwnerScale=true) {module=ShapeLocation001}
   &AddVelocity(
    Cone Axis=<0 0 1>
    Module.UseConeFalloff=false
    Module.UseVelDistribution=false
    Override Seed=false
    Rotation Coordinate Space=::World
    Velocity Mode=::NewEnumerator2
    Velocity Speed=&RandomRangeFloat(Minimum=100, Maximum=125)
   ) {module=AddVelocity002}
   /SetVariables(Particles.DropsVelocity:Vector3f=&Multiply_VectorByFloat(Vector=$Particles.Velocity, Float=0)) {isSetParameters=true}
  ]
  particleUpdateScript [
   &ParticleState()
   &ScaleSpriteSize(Scale Factor=&Vector2DFromFloat(Value=&FloatFromCurve(FloatCurve=DataInterface() {propertyValues {Curve {keys [<at 0 0>,<at 0.5 1>]}}}, Scale Curve=2))) {module=ScaleSpriteSize001}
   &GravityForce(Gravity=<0 0 -600>) {module=GravityForce001}
   &InheritSourceMovement(ApplyForce=false, ApplyPosition=true, ApplyPositionFalloff=false)
   &SolveForcesAndVelocity(Clamp Velocity=false, Limit Acceleration=false) {module=SolveForcesAndVelocity001}
   &ScaleSpriteSize(Scale Factor=&Vector2DFromFloat(Value=&FloatFromCurve(FloatCurve=DataInterface() {propertyValues {Curve {keys [<at 0 0>,<at 1 1>]}}}, Scale Curve=1.5)))
   &DynamicMaterialParameters(
    Index 0 Param 1=&RandomRangeFloat(Minimum=0, Maximum=1)
    Index 0 Param 3=&FloatFromCurve(FloatCurve=DataInterface() {propertyValues {Curve {keys [{interpMode=RCIM_Cubic, tangentMode=RCTM_User, time=0, value=0}, {interpMode=RCIM_Cubic, tangentMode=RCTM_User, time=1, value=1}]}}})
    Index 0 Param 4=&FloatFromCurve(FloatCurve=DataInterface() {propertyValues {Curve {keys [<at 0 0.4 RCIM_Cubic>, {arriveTangent=1.2612, interpMode=RCIM_Cubic, leaveTangent=1.2612, tangentMode=RCTM_User, time=1, value=1}]}}})
    Module.Param0WriteEnabled=true
    Module.Param1WriteEnabled=false
    Module.Param2WriteEnabled=true
    Module.Param3WriteEnabled=true
   ) {module=DynamicMaterialParameters001}
   &ScaleColor(
    Scale Alpha=10
    Scale RGB=&MakeVectorFromLinearColorRGB(Color=&ColorFromCurve(ColorCurve=DataInterface() {propertyValues {
        AlphaCurve {keys [<at 0 1 RCIM_Cubic>]}
        BlueCurve {keys [<at 0 0.0003 RCIM_Cubic>,<at 1 0.007>]}
        GreenCurve {keys [<at 0 0.0018 RCIM_Cubic>,<at 1 0.0469>]}
        RedCurve {keys [<at 0 0.0802 RCIM_Cubic>,<at 1 0.1766>]}
       }}))
    ScaleA=true
    ScaleRGB=true
   )
   /SetVariables(Particles.DropsSize:Vector2f=&Multiply_Vector2D_ByFloat(
     Float=&Multiply_Float(
      A=&RandomRangeFloat(Minimum=1, Maximum=2)
      B=&FloatFromCurve(
       FloatCurve=DataInterface() {propertyValues {Curve {
          defaultValue=0
          keys [{interpMode=RCIM_Cubic, tangentMode=RCTM_User, time=0, value=0}, {interpMode=RCIM_Cubic, tangentMode=RCTM_User, time=1, value=1}]
         }}}
       CurveIndex=&RandomRangeFloat(Minimum=0, Maximum=1)
      )
     )
     Vector2D=<1.5 1.5>
    )) {isSetParameters=true}
  ]
  rendererClasses [&NiagaraSpriteRendererProperties]
  renderers [{rendererClass=&NiagaraSpriteRendererProperties}]
 }
 Sprint_Drops {
  emitterUpdateScript [
   &EmitterState()
   &SpawnPerFrame(
    Spawn Count:NiagaraInt32=&MakeIntFromFloat(Float=&Multiply_Float(
      A=&Lerp_Float(
       A=0
       B=1
       Alpha=&ClampFloat(Float=&Divide_Float(A=&VectorLength(Vector=$Engine.Owner.Velocity), B=500))
      )
      B=$System.SpawnRateCurve
     ))
    Spawn Probability:NiagaraFloat=&MakeCustomFloatFromBool(
     Bool=&SetBoolByFloatComparison(A=2000, B=&VectorLength(Vector=$Engine.Owner.Velocity))
     True Float=0.3
     False Float=0.5
    )
    Use Spawn Probability=true
   ) {module=SpawnPerFrame003}
  ]
  particleSpawnScript [
   &InitializeParticle(
    Color Mode=::NewEnumerator0
    Lifetime=&RandomRangeFloat(Minimum=0.3, Maximum=0.6)
    Sprite Rotation Mode=::NewEnumerator2
    Sprite Size Mode=::NewEnumerator3
    Uniform Sprite Size=&Multiply_Float(
     A=&Multiply_Float(
      A=&RandomRangeFloat(Minimum=1, Maximum=2)
      B=&FloatFromCurve(
       FloatCurve=DataInterface() {propertyValues {Curve {defaultValue=1, keys [{tangentMode=RCTM_User, time=0.8, value=1}, {tangentMode=RCTM_User, time=1, value=1.25}]}}}
       CurveIndex=&RandomRangeFloat(Minimum=0, Maximum=1)
      )
     )
     B=1
    )
    UsePositionOffset=false
    Write Material Random=true
   )
   &CameraOffset(Camera Offset Amount=15) {module=CameraOffset001}
   &SocketLocation(SamplePosOffset=false, Socket Reader=$System.Socket Reader, Socket Sampling Mode=::NewEnumerator3, VelocitySpeedLimit=false)
   &AddVelocity(
    Cone Angle=20
    Cone Axis=<0 0 1>
    Module.UseConeFalloff=false
    Module.UseVelDistribution=false
    Override Seed=false
    Rotation Coordinate Space=::World
    Velocity Mode=::NewEnumerator2
    Velocity Speed=&RandomRangeFloat(Minimum=100, Maximum=150)
   ) {module=AddVelocity001}
   &ShapeLocation(MaskSpawnGroup=false, Offset=<0 0 -12>, Offset Mode=::NewEnumerator0, Override Seed=false, Sphere Radius=1, UseOwnerScale=true)
   /Niagara/Modules/Update/SubUV/SubUVAnimation(SubUV Animation Mode=::NewEnumerator1)
   &AddVelocity(Velocity=<0 -1 0>, Velocity Speed Scale=50) {module=AddVelocity003}
  ]
  particleUpdateScript [
   &ParticleState()
   &ScaleSpriteSize(Scale Factor=&Vector2DFromFloat(Value=&FloatFromCurve(
      FloatCurve=DataInterface() {propertyValues {Curve {
         defaultValue=0
         keys [
          {arriveTangent=9.0725, interpMode=RCIM_Cubic, leaveTangent=9.0725, tangentMode=RCTM_User, time=0, value=0}
          {arriveTangent=0.0926, interpMode=RCIM_Cubic, leaveTangent=0.0926, tangentMode=RCTM_User, time=0.1408, value=0.9847}
          {arriveTangent=-1.146, interpMode=RCIM_Cubic, leaveTangent=-1.146, time=0.7579, value=0.661}
          {arriveTangent=-0.0405, interpMode=RCIM_Cubic, leaveTangent=-0.0405, tangentMode=RCTM_User, time=1, value=0}
         ]
        }}}
      Scale Curve=2
     ))) {module=ScaleSpriteSize001}
   &GravityForce(Gravity=&MakeCustomVectorFromBool(Bool=$User.IsSkydiving, True Vector=<0 0 0>, False Vector=<0 0 -600>)) {module=GravityForce001}
   &InheritSourceMovement(Applied Position Scale=<0.95 0.95 0.95>, ApplyForce=false, ApplyPosition=true, ApplyPositionFalloff=false)
   &SolveForcesAndVelocity(Clamp Velocity=false, Limit Acceleration=false) {module=SolveForcesAndVelocity001}
   &ScaleSpriteSize(Scale Factor=&Vector2DFromFloat(Value=&FloatFromCurve(FloatCurve=DataInterface() {propertyValues {Curve {keys [<at 0 0>,<at 1 1>]}}})))
   &DynamicMaterialParameters(
    Index 0 Param 1=&RandomRangeFloat(Minimum=0, Maximum=1)
    Index 0 Param 3=&FloatFromCurve(FloatCurve=DataInterface() {propertyValues {Curve {keys [{interpMode=RCIM_Cubic, tangentMode=RCTM_User, time=0, value=0}, {interpMode=RCIM_Cubic, tangentMode=RCTM_User, time=1, value=1}]}}})
    Module.Param0WriteEnabled=true
    Module.Param1WriteEnabled=false
    Module.Param2WriteEnabled=true
    Module.Param3WriteEnabled=true
   )
   &ScaleColor(
    Scale Alpha=10
    Scale RGB=&MakeVectorFromLinearColorRGB(Color=&ColorFromCurve(ColorCurve=DataInterface() {propertyValues {
        AlphaCurve {keys [<at 0 1 RCIM_Cubic>]}
        BlueCurve {keys [<at 0.2332 0 RCIM_Cubic>]}
        GreenCurve {keys [<at 0.2332 0.0038 RCIM_Cubic>]}
        RedCurve {keys [<at 0.2332 0.1979 RCIM_Cubic>]}
       }}))
    ScaleA=true
    ScaleRGB=true
   )
  ]
  rendererClasses [&NiagaraSpriteRendererProperties]
  renderers [{rendererClass=&NiagaraSpriteRendererProperties}]
 }
 Sprint_Splash {
  emitterUpdateScript [
   &EmitterState()
   &SpawnPerFrame(
    Spawn Count:NiagaraInt32=&MakeIntFromFloat(Float=&Multiply_Float(
      A=&Lerp_Float(
       A=0
       B=1
       Alpha=&ClampFloat(Float=&Divide_Float(A=&VectorLength(Vector=$Engine.Owner.Velocity), B=500))
      )
      B=$System.SpawnRateCurve
     ))
    Spawn Probability:NiagaraFloat=&MakeCustomFloatFromBool(
     Bool=&SetBoolByFloatComparison(A=2000, B=&VectorLength(Vector=$Engine.Owner.Velocity))
     True Float=0.3
     False Float=0.5
    )
    Use Spawn Probability=true
   ) {module=SpawnPerFrame003}
  ]
  particleSpawnScript [
   &InitializeParticle(
    Color Mode=::NewEnumerator0
    Lifetime=&RandomRangeFloat(Minimum=0.3, Maximum=0.5)
    Sprite Rotation Mode=::NewEnumerator2
    Sprite Size=&Multiply_Vector2D_ByFloat(Float=1, Vector2D=&RandomRangeVector2D(Minimum=<6 6>, Maximum=<9 9>))
    Sprite Size Mode=::NewEnumerator4
    UsePositionOffset=false
    Write Material Random=true
   )
   &CameraOffset(Camera Offset Amount=15)
   &SocketLocation(
    SamplePosOffset=true
    Sampled Position Offset=<0 0 -6>
    Socket Reader=$System.Socket Reader
    Socket Sampling Mode=::NewEnumerator3
    VelocitySpeedLimit=false
   )
   &AddVelocity(
    Cone Angle=20
    Cone Axis=<0 0 1>
    Module.UseConeFalloff=false
    Module.UseVelDistribution=false
    Override Seed=false
    Rotation Coordinate Space=::World
    Velocity Mode=::NewEnumerator2
    Velocity Speed=&RandomRangeFloat(Minimum=80, Maximum=100)
   ) {module=AddVelocity004}
   &ShapeLocation(MaskSpawnGroup=false, Override Seed=false, Sphere Radius=6, UseOwnerScale=true)
   &SubUVAnimation(SubUV Animation Mode=::NewEnumerator1)
   &AddVelocity(Velocity=<0 -1 0>, Velocity Speed Scale=50) {module=AddVelocity003}
  ]
  particleUpdateScript [
   &ParticleState()
   &GravityForce(Gravity=&MakeCustomVectorFromBool(Bool=$User.IsSkydiving, True Vector=<0 0 0>, False Vector=<0 0 -600>))
   &InheritSourceMovement(Applied Position Scale=<0.95 0.95 0.95>, ApplyForce=false, ApplyPosition=true, ApplyPositionFalloff=false)
   &SolveForcesAndVelocity(Clamp Velocity=false, Limit Acceleration=false)
   &ScaleSpriteSize(Scale Factor=&Vector2DFromFloat(Value=&FloatFromCurve(FloatCurve=DataInterface() {propertyValues {Curve {keys [<at 0 0>,<at 0.5 1>]}}}, Scale Curve=1.5)))
   &DynamicMaterialParameters(
    Index 0 Param 1=&RandomRangeFloat(Minimum=0, Maximum=1)
    Index 0 Param 3=&FloatFromCurve(FloatCurve=DataInterface() {propertyValues {Curve {keys [{interpMode=RCIM_Cubic, tangentMode=RCTM_User, time=0, value=0}, {interpMode=RCIM_Cubic, tangentMode=RCTM_User, time=1, value=1}]}}})
    Index 0 Param 4=&FloatFromCurve(FloatCurve=DataInterface() {propertyValues {Curve {keys [<at 0 0.4489 RCIM_Cubic>, {arriveTangent=0.8807, interpMode=RCIM_Cubic, leaveTangent=0.8807, tangentMode=RCTM_User, time=1, value=1}]}}})
    Module.Param0WriteEnabled=true
    Module.Param1WriteEnabled=false
    Module.Param2WriteEnabled=true
    Module.Param3WriteEnabled=true
   ) {module=DynamicMaterialParameters001}
   &ScaleColor(
    Scale Alpha=10
    Scale RGB=&MakeVectorFromLinearColorRGB(Color=&ColorFromCurve(ColorCurve=DataInterface() {propertyValues {
        AlphaCurve {keys [<at 0 1 RCIM_Cubic>]}
        BlueCurve {keys [<at 0.2763 0.0004 RCIM_Cubic>]}
        GreenCurve {keys [<at 0.2763 0.0025 RCIM_Cubic>]}
        RedCurve {keys [<at 0.2763 0.1094 RCIM_Cubic>]}
       }}))
    ScaleA=true
    ScaleRGB=true
   )
  ]
  rendererClasses [&NiagaraSpriteRendererProperties]
  renderers [{rendererClass=&NiagaraSpriteRendererProperties}]
 }
}
summary {
 systemName=NS_Backpack_UskThump_Idle
 emitters [
  {emitterName=Sprint_Splash, rendererClasses [&NiagaraSpriteRendererProperties]}
  {emitterName=Coffee_Waves, rendererClasses [&NiagaraSpriteRendererProperties]}
  {emitterName=Sprint_Drops, rendererClasses [&NiagaraSpriteRendererProperties]}
  {emitterName=Coffee_Splash, rendererClasses [&NiagaraSpriteRendererProperties]}
  {emitterName=Splash, rendererClasses [&NiagaraSpriteRendererProperties]}
 ]
 userVariables [
  User.bPawnLanded {type=NiagaraBool}
  User.bPawnSprintorSlide {type=NiagaraBool}
  User.IsCrouched {type=NiagaraBool}
  User.IsSkydiving {type=NiagaraBool}
  User.IsWeaponFired {type=NiagaraBool}
  User.LandBurst {type=NiagaraFloat, defaultValue=0, description=""}
  User.Switch {type=NiagaraFloat, defaultValue=1, description=""}
 ]
}
