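# Control Rig for the UskThump backpack that keeps its mug bone chain level with the world and adjusts its position by character state.
# World identity rotation is converted to rig space, stripped of Z rotation and applied to base_MugStabilizer, so it cancels tilt but ignores heading.
# After a fixed nudge, the isSkyDiveFloating and isIdle flags, smoothed through AlphaInterp, blend translation offsets onto base_MugOriented and base_MugStabilizer respectively.
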ControlRigBlueprint /BRCosmetics/Accessories/FORT_Backpacks/Backpack_M_MED_UskThump/Meshes/M_MED_UskThump_Pack_CtrlRig
graphs [
 RigVMFunctionLibrary {links [], nodes []}
 RigVMModel {nodes [
   RigUnit_BeginExecution = &RigUnit_BeginExecution()
   From World = /Script/ControlRig.RigUnit_ToRigSpace_Transform(Value=(Rotation=(X=0,Y=0,Z=0,W=1),Translation=(X=0,Y=0,Z=0),Scale3D=(X=1,Y=1,Z=1))) {is=struct RigVMUnitNode, template="From World::Execute(in Value,out Global)"}
   RigVMFunction_MathQuaternionToEuler = &RigVMFunction_MathQuaternionToEuler(RotationOrder=ZYX, Value=@"From World".Global.Rotation)
   RigVMFunction_MathQuaternionFromEuler = &RigVMFunction_MathQuaternionFromEuler(Euler=(X=0,Y=0,Z=0), Euler.X=@RigVMFunction_MathQuaternionToEuler.Result.X, Euler.Y=@RigVMFunction_MathQuaternionToEuler.Result.Y)
   Set Transform = &RigUnit_SetRotation(
    Item=(Type=Bone,Name="base_MugStabilizer")
    Space=GlobalSpace
    Weight=1
    bInitial=False
    ExecutePin=@RigUnit_BeginExecution.ExecutePin
    Value=@RigVMFunction_MathQuaternionFromEuler.Result
   )
   > &RigUnit_ModifyTransforms(
    ItemToModify=((Item=(Type=Bone,Name="base_MugStabilizer"),Transform=(Rotation=(X=0,Y=0,Z=0,W=1),Translation=(X=0,Y=-1,Z=0),Scale3D=(X=1,Y=1,Z=1))))
    Weight=1
    WeightMinimum=0
   )
   RigVMFunction_MathBoolToFloat = &RigVMFunction_MathBoolToFloat(Value=RigVMVariableNode() {variable=isSkyDiveFloating}.Value)
   AlphaInterp = &RigVMFunction_AlphaInterp(InterpSpeedDecreasing=7, InterpSpeedIncreasing=7, Scale=1, bClampResult=False, Value=@RigVMFunction_MathBoolToFloat.Result)
   > &RigUnit_ModifyTransforms(
    ItemToModify=((Item=(Type=Bone,Name="base_MugOriented"),Transform=(Rotation=(X=0,Y=0,Z=0,W=1),Translation=(X=0,Y=14,Z=0),Scale3D=(X=1,Y=1,Z=1))))
    WeightMinimum=0
    Weight=@AlphaInterp.Result
   )
   n1 = &RigVMFunction_AlphaInterp(
    InterpSpeedDecreasing=10
    InterpSpeedIncreasing=8
    Scale=1
    bClampResult=False
    Value=&RigVMFunction_MathBoolToFloat(Value=RigVMVariableNode() {variable=isIdle}.Value).Result
   )
   > &RigUnit_ModifyTransforms(
    ItemToModify=((Item=(Type=Bone,Name="base_MugStabilizer"),Transform=(Rotation=(X=0,Y=0,Z=0,W=1),Translation=(X=0,Y=2.5,Z=0),Scale3D=(X=1,Y=1,Z=1))))
    WeightMinimum=0
    Weight=@n1.Result
   )
  ]}
]
hierarchy [
 root {loc=<0 0 0>, rot=<0 0 0 1>}
 pack {parent=root, loc=<0 -0.7316 0>, rot=<0 0 0 1>}
 base_MugStabilizer {parent=pack, loc=<0 -5.9376 -11.0893>, rot=<0 0 0 1>}
 dyn_MugStabilizer {parent=base_MugStabilizer, loc=<0 0 0>, rot=<0 0 0 1>}
 base_MugOriented {parent=dyn_MugStabilizer, loc=<0.2795 -0.1226 -1.3871>, rot=<-0.0346 0.7926 0.0266 -0.6082>}
 dyn_MugOriented {parent=base_MugOriented, loc=<0 0 0>, rot=<0 0 0 1>}
 base_Coffee {parent=dyn_MugOriented, loc=<15.068 0 0>, rot=<0 -1 0 0>}
 dyn_Coffee {parent=base_Coffee, loc=<0 0 0>, rot=<0 0 0 1>}
]
variables [
 isIdle {type=bool, default=False}
 isSkyDiveFloating {type=bool, default=False}
]
