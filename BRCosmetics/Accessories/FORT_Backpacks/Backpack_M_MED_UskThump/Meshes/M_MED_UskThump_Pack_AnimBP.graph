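# Animation Blueprint for the UskThump backpack cosmetic, posing the pack and its mug stabilizer and simulating coffee motion.
# AnimGraph runs a Control Rig, layers additive bone offsets with blended or interpolated alphas on pack and stabilizer, then applies AnimDynamics to dyn_Coffee.
# EventGraph maps spine bend from the character's mesh into offset and gliding values, and derives idle and directional-movement flags.
# Skydive and parachute events track skydive and hang-glider states, including a delayed float flag.
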AnimBlueprint /BRCosmetics/Accessories/FORT_Backpacks/Backpack_M_MED_UskThump/Meshes/M_MED_UskThump_Pack_AnimBP
target_skeleton=&/M_MED_UskThump_Pack_Skeleton
graphs [
 AnimGraph {nodes [
   n1 = Control Rig!(Source=Input Pose(), Alpha, isSkyDiveFloating:bool=false, isIdle:bool=false) -> (Pose) {controlRigAssetReference {blueprintRigClass=&/M_MED_UskThump_Pack_CtrlRig.M_MED_UskThump_Pack_CtrlRig_C}}
   n2 = "Transform (Modify) Bone - Bone: pack"!(ComponentPose=Local To Component(LocalPose=@n1), Alpha, Translation="0,-1.85,0", Rotation="0,0,0") -> (Pose) {
    rotationMode=BMM_Additive
    translationMode=BMM_Additive
    translationSpace=BCS_BoneSpace
    alphaBoolBlend {blendOption=Linear}
    boneToModify {boneName=pack}
    translation {y=-1.85}
   }
   n3 = "Transform (Modify) Bone - Bone: base_MugStabilizer"!(ComponentPose=@n2, Alpha=1, Translation="0,-2,0", Rotation="0,0,-30") -> (Pose) {
    rotationMode=BMM_Additive
    rotationSpace=BCS_BoneSpace
    translationMode=BMM_Additive
    translationSpace=BCS_BoneSpace
    alphaScaleBiasClamp {bInterpResult=true, interpSpeedDecreasing=8}
    boneToModify {boneName=base_MugStabilizer}
    rotation {roll=-30}
    translation {y=-2}
   }
   n4 = "Transform (Modify) Bone - Bone: base_MugStabilizer"!(ComponentPose=@n3, bAlphaBoolEnabled:bool=True, Translation="0,-4,0", Rotation="0,0,-12") -> (Pose) {
    alphaInputType=Bool
    rotationMode=BMM_Additive
    rotationSpace=BCS_BoneSpace
    translationMode=BMM_Additive
    translationSpace=BCS_BoneSpace
    alphaBoolBlend {blendInTime=0.15, blendOutTime=0.5}
    alphaScaleBiasClamp {bInterpResult=true, interpSpeedDecreasing=8}
    boneToModify {boneName=base_MugStabilizer}
    rotation {roll=-12}
    translation {y=-4}
   }
   n5 = "Transform (Modify) Bone - Bone: base_MugStabilizer"(ComponentPose=@n4, bAlphaBoolEnabled=True, Translation="0,0,-1", Rotation="0,0,0") {
    alphaInputType=Bool
    rotationMode=BMM_Additive
    rotationSpace=BCS_BoneSpace
    translationMode=BMM_Additive
    translationSpace=BCS_BoneSpace
    alphaBoolBlend {blendInTime=0.3, blendOutTime=0.4}
    alphaScaleBiasClamp {bInterpResult=true, interpSpeedDecreasing=8}
    boneToModify {boneName=base_MugStabilizer}
    translation {z=-1}
   }
   n6 = "Transform (Modify) Bone - Bone: base_MugStabilizer"(ComponentPose=@n5, bAlphaBoolEnabled=True, Translation="0,-1.5,0", Rotation="0,0,0") {
    alphaInputType=Bool
    rotationMode=BMM_Additive
    rotationSpace=BCS_BoneSpace
    translationMode=BMM_Additive
    translationSpace=BCS_BoneSpace
    alphaBoolBlend {blendInTime=0.3, blendOutTime=0.4}
    alphaScaleBiasClamp {bInterpResult=true, interpSpeedDecreasing=8}
    boneToModify {boneName=base_MugStabilizer}
    translation {y=-1.5}
   }
   n7 = Anim Dynamics - Bone: dyn_Coffee(ComponentPose=@n6, Alpha=1, ExternalForce="0,0,0") {
    angularDampingOverride=0.9
    angularSpringConstant=10000
    bOverrideAngularDamping=true
    bUseGravityOverride=true
    gravityScale=0.35
    simulationSpace=RootRelative
    boundBone {boneName=dyn_Coffee}
    componentAppliedLinearAccClamp=<9000 9000 9000>
    componentLinearAccScale=<1 1 1>
    componentLinearVelScale=<1 1 1>
    gravityOverride {z=-980}
    physicsBodyDefinitions [{
      boundBone {boneName=dyn_Coffee}
      boxExtents=<20 30 30>
      constraintSetup {angularLimitsMax=<0 14 14>, angularLimitsMin=<0 -14 -14>, angularTarget=<1 0 0>}
      localJointOffset=<2.5 0 0>
     }]
    simSpaceSettings {maxAngularAcceleration=20, maxAngularVelocity=4, simSpaceAngularAlpha=1}
   }
   Output Pose(Result=Component To Local(ComponentPose=@n7)) {layerGroup=None}
  ]}
 EventGraph {nodes [
   Event BlueprintUpdateAnimation()
   > n1 = Sequence!(execute) -> (then_0, then_1, then_2:Exec)
   n2 = Get Skel_Mesh()
   n3 = BreakRotator(InRot=BreakTransform(InTransform=GetDeltaTransformFromRefPose(self=@n2, BoneName=spine_05, BaseName=spine_01)).Rotation)
   > Set backpack_rotate(backpack_rotate=MakeRotator(Roll=MapRangeClamped(Value=@n3.Yaw, InRangeA=0, InRangeB=-30, OutRangeB=-5)))
   > Set backpack_translation(backpack_translation=MakeVector(Y=MapRangeClamped(Value=@n3.Yaw, InRangeA=0, InRangeB=-30, OutRangeB=-4)))
   > Set skyDiveFalling(skyDiveFalling:bool=OR Boolean(A=Get bIsSkydiveDiveMode(), B=Get bIsSkydiveFloating())) -> (Output_Get:bool)
   > n4 = Branch(Condition=Get isHangGliding() -> (isHangGliding:bool))
   n5 = BreakTransform(InTransform=GetDeltaTransformFromRefPose(self=Get Skel_Mesh(), BoneName=spine_04, BaseName=root))
   > Set GlidingBack(GlidingBack:double=MapRangeClamped(Value=BreakRotator(InRot=@n5.Rotation).Yaw, InRangeA=10, InRangeB=-30)) -> (Output_Get:double)
   Set GlidingBack(execute=@n4.else, GlidingBack=0)
   n6 = PlayerMovement(execute=@n1.then_1, AnimInstance=Self-Reference(), XDotProductCoefficient=1, YDotProductCoefficient=1, ZDotProductCoefficient=1)
   n7 = BreakVector(InVec=MakeVector(X=@n6, Y=@n6.YMovement))
   n8 = "Clamp (Float)"(Value=@n7, Min=-1, Max=0)
   > Set movingLeft(movingLeft:bool="To Bool (Double)"!(Value:double="Absolute (Float)"(A=@n8)) -> (ReturnValue:bool)) -> (Output_Get:bool)
   > Set movingRight(movingRight:bool="To Bool (Double)"(Value="Clamp (Float)"(Value=@n7, Min=0, Max=1))) -> (Output_Get:bool)
   > Set MovingForward(MovingForward:bool="To Bool (Double)"(Value="Clamp (Float)"(Value=@n7.Y, Min=0, Max=1))) -> (Output_Get:bool)
   n9 = "Clamp (Float)"(Value=@n7.Y, Min=-1, Max=0)
   > Set MovingBack(MovingBack:bool="To Bool (Double)"(Value="Absolute (Float)"(A=@n9))) -> (Output_Get:bool)
   n10 = Get movingRight() -> (movingRight:bool)
   n11 = Get MovingForward() -> (MovingForward:bool)
   n12 = Get MovingBack() -> (MovingBack:bool)
   n13 = Get bIsPlayingEmote()
   n14 = Get bIsDBNO()
   n15 = Get bIsSkyDiving()
   n16 = Get movingLeft() -> (movingLeft:bool)
   Set isIdle(
    execute=@n1.then_2
    isIdle:bool=NOT Boolean(A=OR Boolean!(A=@n15, B=@n16, C:bool=@n10, D:bool=@n11, E:bool=@n12, F:bool=@n13, G:bool=@n14) -> (ReturnValue))
   ) -> (Output_Get:bool)
   Event OnParachuteOpened()
   > Set startingSkydive(startingSkydive:bool=false) -> (Output_Get:bool)
   > Switch on EGliderType(Selection=Get PlayerGliderType() -> (PlayerGliderType=HangGlider))
   > Set isHangGliding(isHangGliding:bool=true) -> (Output_Get:bool)
   n17 = Event OnCharacterCustomizationLoaded()
   > Set Skel_Mesh(Skel_Mesh=@n17.CustomBaseMesh)
   Event OnBeginSkydiving()
   > Set startingSkydive(startingSkydive=true)
   > Set skydiveFloat(skydiveFloat:bool=false) -> (Output_Get:bool)
   > Delay(Duration=1.5)
   > Set startingSkydive(startingSkydive=false)
   > Set skydiveFloat(skydiveFloat=Get bIsSkydiveFloating())
   Event OnEndSkydiving()
   > Set startingSkydive(startingSkydive=false)
   > Set skydiveFloat(skydiveFloat=false)
   Event OnParachuteClosed()
   > Set skydiveFloat(skydiveFloat=false)
   > Set isHangGliding(isHangGliding=false)
   TryGetPawnOwner()
  ]}
]
variables [
 backpack_rotate {type=Rotator}
 backpack_translation {type=Vector}
 GlidingBack {type=number}
 isHangGliding {type=boolean}
 isIdle {type=boolean}
 MovingBack {type=boolean}
 MovingForward {type=boolean}
 movingLeft {type=boolean}
 movingRight {type=boolean}
 Skel_Mesh {type=SkeletalMeshComponent}
 skyDiveFalling {type=boolean}
 skydiveFloat {type=boolean}
 startingSkydive {type=boolean}
]
